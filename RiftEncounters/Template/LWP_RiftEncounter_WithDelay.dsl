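; Rift encounter template Blueprint that holds back AI spawning for a short delay after the encounter finishes spawning.
; On the post-spawn event it runs the parent logic, waits for the spawn-in delay variable, then allows the spawner to spawn.
; BeginPlay, overlap, tick and the construction script add no logic of their own.
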(event EventBeginPlay
  (|CallParentFunction) ; This node is disabled and will not be called.\nDrag off pins to build functionality.
)

(event Collision|EventActorBeginOverlap
  (params (OtherActor :type Actor%20Object%20Reference)))

(event EventTick
  (params (DeltaSeconds :type Float%20%28single-precision%29)))

(event EventPostSpawnEvents
  ;[1 1 1] After all has been spawned and a small delay, trigger AI to spawn.
  (|CallParentFunction)
  (Utilities|FlowControl|Delay (Variables|Default|GetSpawnInDelay))
  (CallFunction|AllowSpawnertoSpawn))

(fn ConstructionScript
  (|CallParentFunction))
